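# Niagara system for the Ranked Hero umbrella glider, combining a deploy burst with idle effects that follow the glider's sockets.
# Delayed bursts fire a flash, radial wave and staggered hex-grid meshes, while a looping wave, hex sprites and purple spotlights with particle lights persist.
# System RankIndex picks one of eight palettes for all but the spotlights.
# User.ForwardDot speeds up the idle wave and hex spawning; FadeIn ramps idle elements.
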NiagaraSystem /BRCosmetics/Gadgets/Assets/VinderTech_Umbrella/Umbrella_RankedHero/FX/NS_Glider_RankedHero_IdleFX
dependencies {
 usedDataInterfaces [
  &NiagaraDataInterfaceCurve
  &NiagaraDataInterfaceArrayColor
  &NiagaraDataInterfaceColorCurve
  &NiagaraDataInterfaceVectorCurve
  &NiagaraDataInterfaceSocketReader
 ]
 usedDynamicInputs [
  &FN_FrontEndFloatModifier
  &MakeCustomIntFromBool
  &MakeIntFromFloat
  &Subtract_Float
  &Divide_Float
  &MakeFloatFromInt
  /Niagara/DynamicInputs/Divide/Divide_Int
  &MakeCustomLinearColorFromBool
  &Waveform
 ]
 usedModules [
  &self:SystemSpawnScript.SystemScriptSource.SystemScriptGraph.NiagaraNodeAssignment_0.SetVariables_#1
  &self:SystemSpawnScript.SystemScriptSource.SystemScriptGraph.NiagaraNodeAssignment_1.SetVariables_#2
  &self:SystemSpawnScript.SystemScriptSource.SystemScriptGraph.NiagaraNodeAssignment_4.SetVariables_#3
  &SystemState
  &self:SystemSpawnScript.SystemScriptSource.SystemScriptGraph.NiagaraNodeAssignment_5.SetVariables_#4
  &self:SystemSpawnScript.SystemScriptSource.SystemScriptGraph.NiagaraNodeAssignment_6.SetVariables_#5
  &self:SystemSpawnScript.SystemScriptSource.SystemScriptGraph.NiagaraNodeAssignment_2.SetVariables_#6
  &self:SystemSpawnScript.SystemScriptSource.SystemScriptGraph.NiagaraNodeAssignment_3.SetVariables_#7
 ]
}
emitters {
 NE_DeployFlash {
  emitterUpdateScript [
   &EmitterState(Inactive Response=::NewEnumerator1, Life Cycle Mode=::NewEnumerator1, Loop Behavior=::NewEnumerator1, Loop Duration=8.5, UseLoopDelay=false)
   &SpawnBurst_Instantaneous(Spawn Time=$System.BurstDelay, Use Spawn Probability=false, UseLoopCountLimit=false)
  ]
  particleSpawnScript [
   &InitializeParticle(
    Mass Mode=::NewEnumerator1
    Mesh Scale Mode=::NewEnumerator3
    Mesh Uniform Scale=0.2
    Position Mode=::NewEnumerator1
    Position Offset=$System.SystemOffset
    Sprite Rotation Mode=::NewEnumerator2
    Sprite Size=&Vector2DFromFloat(Value=150)
    Sprite Size Mode=::NewEnumerator4
    UsePositionOffset=true
    Write Material Random=true
   ) {module=InitializeParticle001}
   &CameraOffset(Camera Offset Amount=75)
  ]
  particleUpdateScript [
   &ParticleState()
   &SocketLocation(
    SamplePosOffset=true
    Sampled Position Offset=<0 0 18>
    Socket Index=1
    Socket Reader=$System.SocketReader
    Socket Sampling Mode=::NewEnumerator3
    VelocitySpeedLimit=false
   )
   &SolveForcesAndVelocity(Clamp Velocity=false, Limit Acceleration=false)
   &ScaleSpriteSize(Scale Factor=&Multiply_Vector2D_ByFloat(
     Float=&FloatFromCurve(
      FloatCurve=DataInterface() {propertyValues {Curve {keys [
          {arriveTangent=0.7585, interpMode=RCIM_Cubic, leaveTangent=0.7585, tangentMode=RCTM_User, time=0, value=0.4}
          {interpMode=RCIM_Cubic, tangentMode=RCTM_User, time=1, value=1}
         ]}}}
      Scale Curve=&Lerp_Float(A=1, B=3, Alpha=$System.EmpowerAlpha)
     )
     Vector2D=<1 1>
    ))
   &SpriteRotationRate(Rotation Rate=&UniformRangedFloat(Minimum=-50, Maximum=50))
   /SetVariables(Particles.DarkGlowSize:Vector2f=&Multiply_Vector2D_ByFloat(Float=2, Vector2D=$Particles.SpriteSize)) {isSetParameters=true}
   /SetVariables(Particles.CentralFlareSize:Vector2f=&Multiply_Vector2D_ByFloat(Float=3, Vector2D=$Particles.SpriteSize)) {isSetParameters=true}
   /SetVariables(Particles.CentralFlareColor:LinearColor=&MakeLinearColorFromVectorAndFloat(
     "Vector (RGB)"=&Multiply_Vector(
      A=&MakeVectorFromLinearColorRGB(Color=&MergeHelper_ColorSwitch(
        0=&ColorFromCurve(ColorCurve=DataInterface() {propertyValues {
           AlphaCurve {keys [<at 0 1 RCIM_Cubic>,<at 1 0 RCIM_Cubic>]}
           BlueCurve {keys [<at 0 0.25>, {interpMode=RCIM_Cubic, leaveTangent=1.3364, time=1, value=0.1943}]}
           GreenCurve {keys [<at 0 1.7857>, {interpMode=RCIM_Cubic, leaveTangent=-5.7627, time=1, value=0.47}]}
           RedCurve {keys [<at 0 5>, {interpMode=RCIM_Cubic, leaveTangent=-13.364, time=1, value=1}]}
          }})
        1=&ColorFromCurve(ColorCurve=DataInterface() {propertyValues {
           AlphaCurve {keys [<at 0 1 RCIM_Cubic>,<at 1 0 RCIM_Cubic>]}
           BlueCurve {keys [<at 0 5>, {interpMode=RCIM_Cubic, leaveTangent=1.3364, time=1, value=1}]}
           GreenCurve {keys [<at 0 5>, {interpMode=RCIM_Cubic, leaveTangent=-5.7627, time=1, value=1}]}
           RedCurve {keys [<at 0 5>, {interpMode=RCIM_Cubic, leaveTangent=-13.364, time=1, value=1}]}
          }})
        2=&ColorFromCurve(ColorCurve=DataInterface() {propertyValues {
           AlphaCurve {keys [<at 0 1 RCIM_Cubic>,<at 1 0 RCIM_Cubic>]}
           BlueCurve {keys [<at 0 0>, {interpMode=RCIM_Cubic, leaveTangent=1.3364, time=1, value=0}]}
           GreenCurve {keys [<at 0 4.7944>, {interpMode=RCIM_Cubic, leaveTangent=-5.7627, time=1, value=0.5}]}
           RedCurve {keys [<at 0 8>, {interpMode=RCIM_Cubic, leaveTangent=-13.364, time=1, value=1}]}
          }})
        3=&ColorFromCurve(ColorCurve=DataInterface() {propertyValues {
           AlphaCurve {keys [<at 0 1 RCIM_Cubic>,<at 1 0 RCIM_Cubic>]}
           BlueCurve {keys [<at 0 3.875>, {interpMode=RCIM_Cubic, leaveTangent=1.3364, time=1, value=0.8}]}
           GreenCurve {keys [<at 0 5>, {interpMode=RCIM_Cubic, leaveTangent=-5.7627, time=1, value=1}]}
           RedCurve {keys [<at 0 0.5>, {interpMode=RCIM_Cubic, leaveTangent=-13.364, time=1, value=0.2}]}
          }})
        4=&ColorFromCurve(ColorCurve=DataInterface() {propertyValues {
           AlphaCurve {keys [<at 0 1 RCIM_Cubic>,<at 1 0 RCIM_Cubic>]}
           BlueCurve {keys [<at 0 5>, {interpMode=RCIM_Cubic, leaveTangent=1.3364, time=1, value=1}]}
           GreenCurve {keys [<at 0 1.25>, {interpMode=RCIM_Cubic, leaveTangent=-5.7627, time=1, value=0.25}]}
           RedCurve {keys [<at 0 0>, {interpMode=RCIM_Cubic, leaveTangent=-13.364, time=1, value=0}]}
          }})
        5=&ColorFromCurve(ColorCurve=DataInterface() {propertyValues {
           AlphaCurve {keys [<at 0 1 RCIM_Cubic>,<at 1 0 RCIM_Cubic>]}
           BlueCurve {keys [<at 0 4.4487>, <at 0.3217 5>, {interpMode=RCIM_Cubic, leaveTangent=1.3364, time=0.4768, value=1.5}]}
           GreenCurve {keys [<at 0 0>, <at 0.3217 0.0501>, {interpMode=RCIM_Cubic, leaveTangent=-5.7627, time=0.4768, value=0.8951}]}
           RedCurve {keys [<at 0 5>, <at 0.3217 1.0948>, {interpMode=RCIM_Cubic, leaveTangent=-13.364, time=0.4768, value=0.0005}]}
          }})
        6=&ColorFromCurve(ColorCurve=DataInterface() {propertyValues {
           AlphaCurve {keys [<at 0 1 RCIM_Cubic>,<at 1 0 RCIM_Cubic>]}
           BlueCurve {keys [<at 0 0.0321>, {interpMode=RCIM_Cubic, leaveTangent=1.3364, time=0.4044, value=0.0022}]}
           GreenCurve {keys [<at 0 1.9719>, {interpMode=RCIM_Cubic, leaveTangent=-5.7627, time=0.4044, value=0.5937}]}
           RedCurve {keys [<at 0 5>, {interpMode=RCIM_Cubic, leaveTangent=-13.364, time=0.4044, value=5}]}
          }})
        7=&ColorFromCurve(ColorCurve=DataInterface() {propertyValues {
           AlphaCurve {keys [<at 0 1 RCIM_Cubic>,<at 1 0 RCIM_Cubic>]}
           BlueCurve {keys [<at 0 10>, {interpMode=RCIM_Cubic, leaveTangent=1.3364, time=0.4044, value=5}]}
           GreenCurve {keys [<at 0 2.9039>, {interpMode=RCIM_Cubic, leaveTangent=-5.7627, time=0.4044, value=0}]}
           RedCurve {keys [<at 0 0>, {interpMode=RCIM_Cubic, leaveTangent=-13.364, time=0.4044, value=1.2903}]}
          }})
        Index=$System.RankIndex
        Num=::NewEnumerator6
       ))
      B=&MakeVectorFromLinearColorRGB(Color=$System.FEMultiplier)
     )
     "Float (Alpha)"=&FloatFromCurve(
      FloatCurve=DataInterface() {propertyValues {Curve {keys [
          <at 0 0 RCIM_Cubic>
          {arriveTangent=0.3854, interpMode=RCIM_Cubic, leaveTangent=0.3854, tangentMode=RCTM_User, time=0.1, value=1}
          <at 1 0 RCIM_Cubic>
         ]}}}
      Scale Curve=0.2
     )
    )) {isSetParameters=true}
   &ScaleColor(
    Scale Alpha=&FloatFromCurve(
     FloatCurve=DataInterface() {propertyValues {Curve {keys [
         {arriveTangent=6.4019, interpMode=RCIM_Cubic, leaveTangent=6.4019, tangentMode=RCTM_User, time=0, value=1}
         {interpMode=RCIM_Cubic, leaveTangent=0.002, time=1, value=0}
        ]}}}
     Scale Curve=0.05
    )
    Scale RGB=&MakeVectorFromLinearColorRGB(Color=&MultiplyLinearColor(
      Linear Color=$Particles.CentralFlareColor
      Scale Factor=&MakeVector4_FromLinearColor(LinearColor=$System.FEMultiplier)
     ))
    ScaleA=true
    ScaleRGB=true
   )
   &DynamicMaterialParameters(
    Index 1 Param 1=0.3
    Index 1 Param 4=&Lerp_Float(A=0, B=1, Alpha=$System.EmpowerAlpha)
    Module.Index1Float0Write=true
    Module.Index1Float1Write=true
    Module.Index1Float2Write=true
    Module.Index1Float3Write=true
    Module.Index3Float0Write=true
    Module.Index3Float1Write=true
    Module.Index3Float2Write=true
    Module.Index3Float3Write=true
    Write Parameter Index 0=::NewEnumerator0
    Write Parameter Index 1=::NewEnumerator1
    Write Parameter Index 3=::NewEnumerator1
   )
   &FN_Sprite_Attributes(
    Create Sprite Rotation Attribute=true
    Create SpriteSize Attribute=true
    SpriteSize=&Multiply_Vector2D(A=$Particles.CentralFlareSize, B=<2 0.15>)
    Write to Module Namespace=true
   )
  ]
  rendererClasses [&NiagaraSpriteRendererProperties]
  renderers [
   {rendererClass=&NiagaraSpriteRendererProperties}
   {rendererClass=&NiagaraSpriteRendererProperties, rendererIndex=1}
   {rendererClass=&NiagaraSpriteRendererProperties, rendererIndex=2}
  ]
 }
 NE_HexGridDeploy {
  emitterUpdateScript [
   &EmitterState(
    Inactive Response=::NewEnumerator1
    Life Cycle Mode=::NewEnumerator1
    Loop Behavior=::NewEnumerator1
    Loop Duration Mode=::NewEnumerator1
    UseLoopDelay=false
   )
   &SpawnBurst_Instantaneous(Spawn Count=6, Spawn Time=$System.BurstDelay, Use Spawn Probability=false, UseLoopCountLimit=false)
  ]
  particleSpawnScript [
   /SetVariables(Particles.InitialExecIndex:NiagaraFloat=&ReturnNormalizedExecIndex()) {isSetParameters=true}
   &InitializeParticle(
    Color=&SelectLinearColorFromArray(
     Array Sampling Mode=::NewEnumerator1
     Direct Array Index=$System.RankIndex
     Color Selection Array=DataInterface() {propertyValues {ColorData [
        <rgba 1 0.328 0.05 1>
        <rgba 1 1 1 1>
        <rgba 1 0.5 0 1>
        <rgba 0.1 1 0.775 1>
        <rgba 0 0.25 1 1>
        <rgba 0 0.4517 1 1>
        <rgba 1 0.2532 0 1>
        <rgba 0.0047 0.4149 2 1>
       ]}}
     Random Seed=?
    )
    Lifetime=0.5
    Mesh Scale=<0.3 1 1>
    Mesh Scale Mode=::NewEnumerator4
    UsePositionOffset=false
    Write Material Random=true
   )
   &InitialMeshOrientation(Mesh Orientation Mode=::NewEnumerator7, Rotation=<0 -0.25 0>, Use Rotation Vector=true)
   /SetVariables(Particles.Age:NiagaraFloat=&Multiply_Float(A=&OneMinusFloat(Float=$Particles.InitialExecIndex), B=-0.1)) {isSetParameters=true}
  ]
  particleUpdateScript [
   &ParticleState()
   &ScaleMeshSize(Scale Factor=&Multiply_VectorByFloat(
     Vector=<1 1 1>
     Float=&Multiply_Float(
      A=&FloatFromCurve(FloatCurve=DataInterface() {propertyValues {Curve {keys [
           {arriveTangent=1.2481, interpMode=RCIM_Cubic, leaveTangent=1.2481, tangentMode=RCTM_User, time=0, value=0}
           {arriveTangent=0.7346, interpMode=RCIM_Cubic, leaveTangent=0.7346, tangentMode=RCTM_User, time=1, value=1}
          ]}}})
      B=&Lerp_Float(
       A=0.1
       B=0.9
       Alpha=&FloatFromCurve(
        FloatCurve=DataInterface() {propertyValues {Curve {keys [
            {arriveTangent=1.5034, interpMode=RCIM_Cubic, leaveTangent=1.5034, tangentMode=RCTM_User, time=0, value=0}
            {arriveTangent=0.3999, interpMode=RCIM_Cubic, leaveTangent=0.3999, tangentMode=RCTM_User, time=1, value=1}
           ]}}}
        CurveIndex=$Particles.InitialExecIndex
       )
      )
     )
    ))
   &SocketLocation(
    Position Sampling=::NewEnumerator1
    SamplePosOffset=true
    Sampled Position Offset=<0 0 19>
    Socket Index=1
    Socket Reader=$System.SocketReader
    Socket Sampling Mode=::NewEnumerator3
    VelocitySpeedLimit=false
   )
   /SetVariables(Particles.Position=&Lerp_Vector(
     A=$Output.SocketLocation.SampledPosition
     B=&AddVectorToPosition(
      Position=$Output.SocketLocation.SampledPosition
      Vector=&Lerp_Vector(
       A=<0 0 -5>
       B=<0 0 -25>
       Alpha=&FloatFromCurve(
        FloatCurve=DataInterface() {propertyValues {Curve {keys [
            {arriveTangent=0.3354, interpMode=RCIM_Cubic, leaveTangent=0.3354, tangentMode=RCTM_User, time=0, value=0}
            {arriveTangent=1.9286, interpMode=RCIM_Cubic, leaveTangent=1.9286, tangentMode=RCTM_User, time=1, value=1}
           ]}}}
        CurveIndex=$Particles.InitialExecIndex
       )
      )
     )
     Alpha=&FloatFromCurve(FloatCurve=DataInterface() {propertyValues {Curve {keys [
          {arriveTangent=0.9109, interpMode=RCIM_Cubic, leaveTangent=0.9109, tangentMode=RCTM_User, time=0, value=0}
          {arriveTangent=1.3677, interpMode=RCIM_Cubic, leaveTangent=1.3677, tangentMode=RCTM_User, time=1, value=1}
         ]}}})
    )) {isSetParameters=true}
   &DynamicMaterialParameters(Module.Index0Float0Write=true, Module.Index0Float1Write=true, Module.Index0Float2Write=true, Module.Index0Float3Write=true)
   /SetVariables(Particles.MeshOrientation=&MultiplyQuaternion(
     Quaternion A=$Particles.MeshOrientation
     Quaternion B=&MakeQuaternion(
      Quaternion From=0
      Axis=<1 0 0>
      Angle=&FloatFromCurve(
       FloatCurve=DataInterface() {propertyValues {Curve {keys [<at 0 0>,<at 1 1>]}}}
       Scale Curve=&Lerp_Float(A=0.5, B=0.25, Alpha=$Particles.InitialExecIndex)
      )
      From=?
      To=?
      Angle Type=::NewEnumerator1
      Coordinate Space=::Local
      Normalize Vectors=false
     )
    )) {isSetParameters=true}
   &ScaleColor(
    Scale Alpha=&FloatFromCurve(FloatCurve=DataInterface() {propertyValues {Curve {keys [
         {arriveTangent=-0.4448, interpMode=RCIM_Cubic, leaveTangent=-0.4448, tangentMode=RCTM_User, time=0, value=1}
         {arriveTangent=0.0452, interpMode=RCIM_Cubic, leaveTangent=0.0452, tangentMode=RCTM_User, time=1, value=0}
        ]}}})
    ScaleA=true
    ScaleRGB=true
   )
  ]
  rendererClasses [&NiagaraMeshRendererProperties]
  renderers [{rendererClass=&NiagaraMeshRendererProperties, rendererIndex=0}]
 }
 NE_HexSprites {
  emitterUpdateScript [
   &EmitterState(
    Inactive Response=::NewEnumerator1
    Life Cycle Mode=::NewEnumerator1
    Loop Behavior=::NewEnumerator1
    Loop Duration Mode=::NewEnumerator1
    UseLoopDelay=false
   )
   &SpawnRate(
    SpawnGroup=1
    SpawnRate=&Multiply_Float(A=&Lerp_Float(A=5, B=15, Alpha=$User.ForwardDot), B=$System.FadeIn)
    Use Spawn Probability=false
   )
   &SpawnBurst_Instantaneous(Spawn Count=15, Spawn Time=$System.BurstDelay, Use Spawn Probability=false, UseLoopCountLimit=false)
  ]
  particleSpawnScript [
   &InitializeParticle(
    Color=&MultiplyLinearColor(
     Linear Color=&MergeHelper_ColorSwitch(
      0=<rgba 1 0.2927 0 1>
      1=<rgba 1 1 1 1>
      2=<rgba 1 0.5 0 1>
      3=<rgba 0.1 1 0.775 1>
      4=<rgba 0 0.25 1 1>
      5=&ColorFromCurve(
       ColorCurve=DataInterface() {propertyValues {
         AlphaCurve {keys [<at 0 1 RCIM_Cubic>]}
         BlueCurve {keys [<at 0.4259 1 RCIM_Cubic>,<at 0.4497 1 RCIM_Cubic>]}
         GreenCurve {keys [<at 0.4259 0.0031 RCIM_Cubic>,<at 0.4497 0.4906 RCIM_Cubic>]}
         RedCurve {keys [<at 0.4259 0.9694 RCIM_Cubic>,<at 0.4497 0 RCIM_Cubic>]}
        }}
       CurveIndex=&RandomRangeFloat(Minimum=0, Maximum=1)
      )
      6=&ColorFromCurve(
       ColorCurve=DataInterface() {propertyValues {
         AlphaCurve {keys [<at 0 1 RCIM_Cubic>]}
         BlueCurve {keys [<at 0.4259 0 RCIM_Cubic>,<at 0.4497 0 RCIM_Cubic>]}
         GreenCurve {keys [<at 0.4259 0.3245 RCIM_Cubic>,<at 0.4497 0.0937 RCIM_Cubic>]}
         RedCurve {keys [<at 0.4259 1 RCIM_Cubic>,<at 0.4497 1 RCIM_Cubic>]}
        }}
       CurveIndex=&RandomRangeFloat(Minimum=0, Maximum=1)
      )
      7=&ColorFromCurve(
       ColorCurve=DataInterface() {propertyValues {
         AlphaCurve {keys [<at 0 1 RCIM_Cubic>]}
         BlueCurve {keys [<at 0.4259 1 RCIM_Cubic>,<at 0.4497 1 RCIM_Cubic>]}
         GreenCurve {keys [<at 0.4259 0.2213 RCIM_Cubic>,<at 0.4497 0 RCIM_Cubic>]}
         RedCurve {keys [<at 0.4259 0 RCIM_Cubic>,<at 0.4497 0.3542 RCIM_Cubic>]}
        }}
       CurveIndex=&RandomRangeFloat(Minimum=0, Maximum=1)
      )
      Index=$System.RankIndex
      Num=::NewEnumerator6
     )
     Scale Factor=&MakeVector4_FromLinearColor(LinearColor=&RandomRangeLinearColor(Minimum=<rgba 2 2 2 1>, Maximum=<rgba 5 5 5 1>))
    )
    Lifetime=&MaskFloatBySpawnGroup(
     Masked Float=&RandomRangeFloat(Minimum=0.5, Maximum=0.6)
     Passthrough Float=&RandomRangeFloat(Minimum=0.75, Maximum=1)
    )
    Sprite Size Mode=::NewEnumerator3
    Uniform Sprite Size=&MaskFloatBySpawnGroup(
     Masked Float=&RandomRangeFloat(Minimum=5, Maximum=10)
     Passthrough Float=&RandomRangeFloat(Minimum=5, Maximum=6)
    )
    UsePositionOffset=false
    Write Material Random=true
   )
   &SocketLocation(
    SamplePosOffset=true
    Sampled Position Offset=<0 0 -5>
    Socket Index=1
    Socket Reader=$System.SocketReader
    Socket Sampling Mode=::NewEnumerator3
    VelocitySpeedLimit=false
   )
   &ShapeLocation(
    Cylinder Distribution=::NewEnumerator2
    Cylinder Height=5
    Cylinder Radius=&MaskFloatBySpawnGroup(Masked Float=&RandomRangeFloat(Minimum=5, Maximum=35), Passthrough Float=55)
    MaskSpawnGroup=false
    Shape Primitive=::NewEnumerator1
    U Position=&MaskFloatBySpawnGroup(
     Masked Float=&RandomRangeFloat(Minimum=0, Maximum=1)
     Passthrough Float=&RandomRangeFloat(Minimum=0, Maximum=1)
    )
    UseOwnerScale=true
    V Position=1
   )
   &AddVelocity(
    Velocity=&NormalizeVector(Vector To Normalize=$Particles.ShapeLocation.ShapeVector, Safe Normalize=true, Safe Fallback Threshold=0, Safe Fallback Vector=<0 0 0.001>)
    Velocity Speed Scale=&MaskFloatBySpawnGroup(
     Masked Float=&RandomRangeFloat(Minimum=50, Maximum=450)
     Passthrough Float=&RandomRangeFloat(Minimum=50, Maximum=75)
    )
   )
   &AddVelocity(
    Velocity=<0 0 1>
    Velocity Speed Scale=&MaskFloatBySpawnGroup(
     Masked Float=&RandomRangeFloat(Minimum=50, Maximum=150)
     Passthrough Float=&RandomRangeFloat(Minimum=50, Maximum=75)
    )
   ) {module=AddVelocity001}
   /SetVariables(Particles.SubImageIndex=&RandomRangeFloat(Minimum=4, Maximum=5)) {isSetParameters=true}
   &CameraOffset()
  ]
  particleUpdateScript [
   &ParticleState()
   &ScaleSpriteSize(Uniform Curve Sprite Scale=DataInterface() {propertyValues {Curve {keys [
        {arriveTangent=9.4436, interpMode=RCIM_Cubic, leaveTangent=9.4436, tangentMode=RCTM_User, time=0, value=0}
        {arriveTangent=0.0374, interpMode=RCIM_Cubic, leaveTangent=0.0374, tangentMode=RCTM_User, time=0.3, value=2}
        {arriveTangent=-1.1378, interpMode=RCIM_Cubic, leaveTangent=-1.1378, tangentMode=RCTM_User, time=1, value=1}
       ]}}})
   &Drag(
    Drag=&MaskFloatBySpawnGroup(Masked Float=3, Passthrough Float=1)
    Use Linear Drag=true
    Use Rotational Drag=true
   )
   &InheritSourceMovement(
    Applied Position Scale=&Multiply_VectorByFloat(
     Vector=&MaskVectorBySpawnGroup(Masked Vector=<1 1 0.3>, Passthrough Vector=<1 1 0.1>)
     Float=&FloatFromCurve(
      FloatCurve=DataInterface() {propertyValues {Curve {keys [
          {arriveTangent=1.9525, interpMode=RCIM_Cubic, leaveTangent=1.9525, tangentMode=RCTM_User, time=0, value=0}
          {arriveTangent=-0.0252, interpMode=RCIM_Cubic, leaveTangent=-0.0252, tangentMode=RCTM_User, time=1, value=1}
         ]}}}
      Scale Curve=&RandomRangeFloat(Minimum=-0.1, Maximum=-0.2)
     )
    )
    ApplyForce=false
    ApplyPosition=true
    ApplyPositionFalloff=false
    Source Velocity:Vector3f=&TransformVector(Vector=$Engine.Owner.Velocity, Source Space=::World, Destination Space=::Local)
   )
   &SolveForcesAndVelocity(Clamp Velocity=false, Limit Acceleration=false)
   &ScaleColor(
    Scale Alpha=&FloatFromCurve(FloatCurve=DataInterface() {propertyValues {Curve {keys [
         {interpMode=RCIM_Cubic, tangentMode=RCTM_User, time=0, value=1}
         {arriveTangent=-1.7686, interpMode=RCIM_Cubic, leaveTangent=-1.7686, tangentMode=RCTM_User, time=1, value=0}
        ]}}})
    Scale RGB=&VectorFromFloat(Value=&Lerp_Float(
      A=1
      B=10
      Alpha=&FloatFromCurve(FloatCurve=DataInterface() {propertyValues {Curve {keys [
           {arriveTangent=-2.03, interpMode=RCIM_Cubic, leaveTangent=-2.03, tangentMode=RCTM_User, time=0, value=1}
           {interpMode=RCIM_Cubic, tangentMode=RCTM_User, time=1, value=0}
          ]}}})
     ))
    ScaleA=true
    ScaleRGB=true
   )
  ]
  rendererClasses [&NiagaraSpriteRendererProperties]
  renderers [{rendererClass=&NiagaraSpriteRendererProperties}]
 }
 NE_SpotLights {
  emitterUpdateScript [
   &EmitterState(
    Inactive Response=::NewEnumerator1
    Life Cycle Mode=::NewEnumerator1
    Loop Behavior=::NewEnumerator1
    Loop Duration Mode=::NewEnumerator1
    UseLoopDelay=false
   )
   &SpawnBurst_Instantaneous(Spawn Count=$System.SpotNumber, Spawn Time=$System.BurstDelay, Use Spawn Probability=false, UseLoopCountLimit=false)
  ]
  particleSpawnScript [
   &PartitionParticles(
    ID to Partition:NiagaraInt32=&ReturnExecIndex()
    Particles Per Partition:NiagaraInt32=/Niagara/DynamicInputs/Subtract/Subtract_Integer(A:NiagaraInt32=$System.SpotNumber, B:NiagaraInt32=1)
    Partition Mode:ENiagara_PartitionMode_v2=::NewEnumerator1
   )
   &InitializeParticle(
    Color=&MaskLinearColor(
     Mask Source=1
     Mask Parameter=$Particles.PartitionParticles.Partition
     Mask=0
     Masked Linear Color=<rgba 0.2966 0 1 0.3>
     Passthrough Linear Color=<rgba 0.7233 0 2 2>
    )
    Sprite Size=/Niagara/DynamicInputs/Helpers/MaskVector2d(
     Mask Source:NiagaraInt32=1
     Mask Parameter:NiagaraInt32=$Particles.PartitionParticles.Partition
     Mask:NiagaraInt32=0
     Masked Vector2D:Vector2f=<15 20>
     Passthrough Vector2D:Vector2f=<30 15>
    )
    Sprite Size Mode=::NewEnumerator4
    UsePositionOffset=false
    Write Material Random=true
   )
   /SetVariables(Particles.SubImageIndex=2) {isSetParameters=true}
  ]
  particleUpdateScript [
   &ParticleState(Kill Particles When Lifetime Has Elapsed=false)
   /SetVariables(Particles.BoostAlpha:NiagaraFloat=&FloatFromCurve(FloatCurve=DataInterface() {propertyValues {Curve {keys [
         {arriveTangent=10.3169, interpMode=RCIM_Cubic, leaveTangent=10.3169, tangentMode=RCTM_User, time=0, value=0}
         {interpMode=RCIM_Cubic, tangentMode=RCTM_User, time=0.1, value=1}
         {interpMode=RCIM_Cubic, tangentMode=RCTM_User, time=0.5, value=0}
        ]}}})) {isSetParameters=true}
   &ScaleSpriteSize(
    Scale Factor=&MakeVector2D(X=1, Y=&Lerp_Float(A=1, B=1, Alpha=$Particles.BoostAlpha))
    Scale Sprite Size Mode=::NewEnumerator2
   )
   &SocketLocation(
    Position Sampling=::NewEnumerator1
    SamplePosOffset=true
    Sampled Position Offset=<0 0 16>
    Socket Reader=$System.SocketReader
    Socket Sampling Mode=::NewEnumerator3
    VelocitySpeedLimit=false
   )
   &RotateAroundPoint(
    Radius=6.5
    Rotation Center:NiagaraPosition=&Add_Vector(A=Linked() {linkedVariable {name=Output.SocketLocation.SampledPosition, type=NiagaraPosition}}, B=<0 0 1.5>)
    Rotation Phase:NiagaraFloat=$Particles.PartitionParticles.Normalized Index In Partition
    Rotation Rate=0.2
    Set Particle Position=false
   )
   /SetVariables(Particles.Position=/Niagara/DynamicInputs/Helpers/MaskPosition(
     Mask Source:NiagaraInt32=1
     Mask Parameter:NiagaraInt32=$Particles.PartitionParticles.Partition
     Mask:NiagaraInt32=0
     Masked Position:NiagaraPosition=$Output.RotateAroundPoint.OutputPosition
     Passthrough Position:NiagaraPosition=$Output.SocketLocation.SampledPosition
    )) {isSetParameters=true}
   &SpriteFacingAndAlignment(
    Sprite Alignment=&MaskVector(
     Mask Source=1
     Mask Parameter=$Particles.PartitionParticles.Partition
     Mask=0
     Masked Vector=&Add_Vector(
      A=&NormalizeVector(
       Vector To Normalize=$Particles.RotateAroundPoint.VectorFromCenter
       Safe Normalize=true
       Safe Fallback Threshold=0
       Safe Fallback Vector=<0 0 0.001>
      )
      B=&Lerp_Vector(A=<0 0 7>, B=<0 0 1>, Alpha=$Particles.BoostAlpha)
     )
     Passthrough Vector=<0 0 1>
    )
    WriteAlignment=true
    WriteFacing=false
   )
   &ScaleColor(
    Scale Alpha=&Multiply_Float(A=&Add_Float(A=1, B=$System.Oscillation), B=1)
    Scale RGB=&VectorFromFloat(Value=&Lerp_Float(A=1, B=5, Alpha=$Particles.BoostAlpha))
    ScaleA=true
    ScaleRGB=true
   )
   &FN_Sprite_Attributes(
    Color=&Multiply_LinearColorByFloat(Linear Color=<rgba 2.5556 0.7805 5 1>, Float=$System.FadeIn)
    Create Color Attribute=true
    Create Position Attribute=true
    Create SpriteSize Attribute=true
    Position:NiagaraPosition=&AddVectorToPosition(
     Position=$Particles.Position
     Vector=&MaskVector(Mask Source=1, Mask Parameter=$Particles.PartitionParticles.Partition, Mask=0, Masked Vector=<0 0 1>, Passthrough Vector=<0 0 1>)
    )
    SpriteSize=&Vector2DFromFloat(Value=&MaskFloat(Mask Source=1, Mask Parameter=$Particles.PartitionParticles.Partition, Mask=0, Masked Float=10, Passthrough Float=30))
    Write to Module Namespace=true
   )
   &FN_Light_Attributes(
    Color=<rgba 0.4356 0.1203 1 1>
    Create Color Attribute=true
    Create Exponent Attribute=true
    Create Position Attribute=true
    Create Radius Attribute=true
    Enable Particle Light Enable=true
    Exponent=1
    Light Enabled:NiagaraBool=&SetBoolByIntComparison(
     A=&ReturnExecIndex()
     B=&Subtract_Integer(A:NiagaraInt32=$System.SpotNumber, B:NiagaraInt32=2)
     Comparison Type=::NewEnumerator0
    )
    Position=&AddVectorToPosition(Position=$Output.SocketLocation.SampledPosition, Vector=<0 0 7>)
    Radius=50
    Scale Color=&Multiply_Float(A=3, B=$System.FadeIn)
   )
  ]
  rendererClasses [&NiagaraSpriteRendererProperties, &NiagaraLightRendererProperties]
  renderers [
   {rendererClass=&NiagaraSpriteRendererProperties}
   {rendererClass=&NiagaraSpriteRendererProperties, rendererIndex=1}
   {rendererClass=&NiagaraLightRendererProperties, rendererIndex=2}
  ]
 }
 NE_WaveDeploy {
  emitterUpdateScript [
   &EmitterState(
    Inactive Response=::NewEnumerator1
    Life Cycle Mode=::NewEnumerator1
    Loop Behavior=::NewEnumerator1
    Loop Duration Mode=::NewEnumerator1
    UseLoopDelay=false
   )
   &SpawnBurst_Instantaneous(Spawn Time=$System.BurstDelay, Use Spawn Probability=false, UseLoopCountLimit=false)
  ]
  particleSpawnScript [&InitializeParticle(
    Color Mode=::NewEnumerator0
    Lifetime=0.5
    Mesh Scale=<1.5 1 1.5>
    Mesh Scale Mode=::NewEnumerator4
    UsePositionOffset=false
    Write Material Random=true
   )]
  particleUpdateScript [
   &ParticleState()
   &FN_Radial_Material(Base Sprite Tile U=3, Depth Fade=0, Sprite Extrusion Speed=0.2, Sprite Spin Speed=0.1)
   &SocketLocation(SamplePosOffset=true, Socket Index=1, Socket Reader=$System.SocketReader, Socket Sampling Mode=::NewEnumerator3, VelocitySpeedLimit=false)
   &Color(Color=&MultiplyLinearColor(
     Linear Color=&MergeHelper_ColorSwitch(
      0=&ColorFromCurve(ColorCurve=DataInterface() {propertyValues {
         AlphaCurve {keys [<at 0 1 RCIM_Cubic>,<at 1 0 RCIM_Cubic>]}
         BlueCurve {keys [<at 0 0.5>, {interpMode=RCIM_Cubic, leaveTangent=1.3364, time=1, value=0.1943}]}
         GreenCurve {keys [<at 0 3.5713>, {interpMode=RCIM_Cubic, leaveTangent=-5.7627, time=1, value=0.47}]}
         RedCurve {keys [<at 0 10>, {interpMode=RCIM_Cubic, leaveTangent=-13.364, time=1, value=1}]}
        }})
      1=&ColorFromCurve(ColorCurve=DataInterface() {propertyValues {
         AlphaCurve {keys [<at 0 1 RCIM_Cubic>,<at 1 0 RCIM_Cubic>]}
         BlueCurve {keys [<at 0 10>, {interpMode=RCIM_Cubic, leaveTangent=1.3364, time=1, value=1}]}
         GreenCurve {keys [<at 0 10>, {interpMode=RCIM_Cubic, leaveTangent=-5.7627, time=1, value=1}]}
         RedCurve {keys [<at 0 10>, {interpMode=RCIM_Cubic, leaveTangent=-13.364, time=1, value=1}]}
        }})
      2=&ColorFromCurve(ColorCurve=DataInterface() {propertyValues {
         AlphaCurve {keys [<at 0 1 RCIM_Cubic>,<at 1 0 RCIM_Cubic>]}
         BlueCurve {keys [<at 0 1>, {interpMode=RCIM_Cubic, leaveTangent=1.3364, time=1, value=0.1}]}
         GreenCurve {keys [<at 0 5.5>, {interpMode=RCIM_Cubic, leaveTangent=-5.7627, time=1, value=0.55}]}
         RedCurve {keys [<at 0 10>, {interpMode=RCIM_Cubic, leaveTangent=-13.364, time=1, value=1}]}
        }})
      3=&ColorFromCurve(ColorCurve=DataInterface() {propertyValues {
         AlphaCurve {keys [<at 0 1 RCIM_Cubic>,<at 1 0 RCIM_Cubic>]}
         BlueCurve {keys [<at 0 7.75>, {interpMode=RCIM_Cubic, leaveTangent=1.3364, time=1, value=0.8}]}
         GreenCurve {keys [<at 0 10>, {interpMode=RCIM_Cubic, leaveTangent=-5.7627, time=1, value=1}]}
         RedCurve {keys [<at 0 1>, {interpMode=RCIM_Cubic, leaveTangent=-13.364, time=1, value=0.2}]}
        }})
      4=&ColorFromCurve(ColorCurve=DataInterface() {propertyValues {
         AlphaCurve {keys [<at 0 1 RCIM_Cubic>,<at 1 0 RCIM_Cubic>]}
         BlueCurve {keys [<at 0 10>, {interpMode=RCIM_Cubic, leaveTangent=1.3364, time=1, value=1}]}
         GreenCurve {keys [<at 0 2.5>, {interpMode=RCIM_Cubic, leaveTangent=-5.7627, time=1, value=0.25}]}
         RedCurve {keys [<at 0 0>, {interpMode=RCIM_Cubic, leaveTangent=-13.364, time=1, value=0}]}
        }})
      5=&ColorFromCurve(ColorCurve=DataInterface() {propertyValues {
         AlphaCurve {keys [<at 0 1 RCIM_Cubic>,<at 1 0 RCIM_Cubic>]}
         BlueCurve {keys [<at 0 8.8973>, <at 0.3217 7>, {interpMode=RCIM_Cubic, leaveTangent=1.3364, time=0.4768, value=2}]}
         GreenCurve {keys [<at 0 0>, <at 0.3217 0.0702>, {interpMode=RCIM_Cubic, leaveTangent=-5.7627, time=0.4768, value=1.1935}]}
         RedCurve {keys [<at 0 10>, <at 0.3217 1.5327>, {interpMode=RCIM_Cubic, leaveTangent=-13.364, time=0.4768, value=0.0007}]}
        }})
      6=&ColorFromCurve(ColorCurve=DataInterface() {propertyValues {
         AlphaCurve {keys [<at 0 1 RCIM_Cubic>,<at 1 0 RCIM_Cubic>]}
         BlueCurve {keys [<at 0 0>,<at 0.4095 0>]}
         GreenCurve {keys [<at 0 3.534>,<at 0.4095 0.6552>]}
         RedCurve {keys [<at 0 10>,<at 0.4095 6.9742>]}
        }})
      7=&ColorFromCurve(ColorCurve=DataInterface() {propertyValues {
         AlphaCurve {keys [<at 0 1 RCIM_Cubic>,<at 1 0 RCIM_Cubic>]}
         BlueCurve {keys [<at 0 15>,<at 0.1959 11.6522>,<at 0.4095 8>]}
         GreenCurve {keys [<at 0 3.2325>,<at 0.1959 0.0049>,<at 0.4095 0>]}
         RedCurve {keys [<at 0 0>,<at 0.1959 0>,<at 0.4095 2.4869>]}
        }})
      Index=$System.RankIndex
      Num=::NewEnumerator6
     )
     Scale Factor=&MakeVector4FromVectorAndFloat(
      "Vector (RGB)"=<1 1 1>
      "Float (Alpha)"=&FloatFromCurve(FloatCurve=DataInterface() {propertyValues {Curve {keys [
           <at 0 0 RCIM_Cubic>
           {interpMode=RCIM_Cubic, tangentMode=RCTM_User, time=0.1, value=1}
           {arriveTangent=-1.6495, interpMode=RCIM_Cubic, leaveTangent=-1.6495, tangentMode=RCTM_User, time=1, value=0}
          ]}}})
     )
    ))
   /SetVariables(Particles.RadialHexScaling:Vector4f=&MakeVector4(X=7, Y=10, Z=10, W=2.8725)) {isSetParameters=true, module=SetVariables_#8}
   &ScaleMeshSize(Scale Factor=&VectorFromCurve(VectorCurve=DataInterface() {propertyValues {
       XCurve {keys [
         {arriveTangent=1.1565, interpMode=RCIM_Cubic, leaveTangent=1.1565, tangentMode=RCTM_User, time=0, value=0.5}
         {interpMode=RCIM_Cubic, tangentMode=RCTM_User, time=1, value=1}
        ]}
       YCurve {keys [<at 0 1>,<at 1 1>]}
       ZCurve {keys [
         {arriveTangent=1.1565, interpMode=RCIM_Cubic, leaveTangent=1.1565, tangentMode=RCTM_User, time=0, value=0.5}
         {interpMode=RCIM_Cubic, tangentMode=RCTM_User, time=1, value=1}
        ]}
      }}))
  ]
  rendererClasses [&NiagaraMeshRendererProperties]
  renderers [{rendererClass=&NiagaraMeshRendererProperties, rendererIndex=0}]
 }
 NE_WaveIdle {
  emitterSpawnScript [/SetVariables() {isSetParameters=true}]
  emitterUpdateScript [
   &EmitterState(
    Inactive Response=::NewEnumerator1
    Life Cycle Mode=::NewEnumerator1
    Loop Behavior=::NewEnumerator1
    Loop Duration Mode=::NewEnumerator1
    UseLoopDelay=false
   )
   &SpawnBurst_Instantaneous(Spawn Time=$System.BurstDelay, Use Spawn Probability=false, UseLoopCountLimit=false)
   /SetVariables(Emitter.CustomTime:NiagaraFloat=&Add_Float(A=$Emitter.CustomTime, B=&Lerp_Float(A=0.03, B=0.05, Alpha=$User.ForwardDot))) {isSetParameters=true}
  ]
  particleSpawnScript [&InitializeParticle(Color Mode=::NewEnumerator0, Lifetime=5, UsePositionOffset=false, Write Material Random=true)]
  particleUpdateScript [
   &ParticleState(Kill Particles When Lifetime Has Elapsed=false)
   &FN_Radial_Material(Base Sprite Tile U=3, Depth Fade=20, Sprite Extrusion Speed=0.2, Sprite Spin Speed=0.1, TIme:NiagaraFloat=$Emitter.CustomTime)
   &SocketLocation(SamplePosOffset=true, Socket Index=1, Socket Reader=$System.SocketReader, Socket Sampling Mode=::NewEnumerator3, VelocitySpeedLimit=false)
   &Color(Color=&Multiply_LinearColorByFloat(
     Linear Color=&SelectLinearColorFromArray(
      Array Sampling Mode=::NewEnumerator1
      Direct Array Index=$System.RankIndex
      Color Selection Array=DataInterface() {propertyValues {ColorData [
         <rgba 1 0.3988 0.15 1>
         <rgba 1 1 1 1>
         <rgba 1 0.55 0.1 1>
         <rgba 0.2772 1 0.8263 1>
         <rgba 0 0.25 1 1>
         <rgba 0 1 1 1>
         <rgba 1 0.3305 0.0292 1>
         <rgba 0 0.2 2 1>
        ]}}
      Random Seed=?
     )
     Float=$System.FadeIn
    ))
   /SetVariables(Particles.RadialHexScaling:Vector4f=&MakeVector4(X=7, Y=10, Z=10, W=2.8725)) {isSetParameters=true, module=SetVariables_#8}
   /SetVariables(Particles.HexColor:LinearColor=&Multiply_LinearColorByFloat(
     Linear Color=&SelectLinearColorFromArray(
      Array Sampling Mode=::NewEnumerator1
      Direct Array Index=$System.RankIndex
      Color Selection Array=DataInterface() {propertyValues {ColorData [
         <rgba 1 0.3988 0.15 1>
         <rgba 0.474 0.474 0.474 1>
         <rgba 1 0.55 0.1 1>
         <rgba 0.1 1 0.7837 1>
         <rgba 0 0.25 1 1>
         <rgba 0.849 0.1047 0.5069 1>
         <rgba 1 0.1064 0 1>
         <rgba 0.6135 0 1.5 1>
        ]}}
      Random Seed=?
     )
     Float=$System.FadeIn
    )) {isSetParameters=true}
  ]
  rendererClasses [&NiagaraMeshRendererProperties]
  renderers [
   {rendererClass=&NiagaraMeshRendererProperties, rendererIndex=0}
   {rendererClass=&NiagaraMeshRendererProperties, rendererIndex=1}
  ]
 }
}
summary {
 systemName=NS_Glider_RankedHero_IdleFX
 emitters [
  {emitterName=NE_SpotLights, rendererClasses [&NiagaraSpriteRendererProperties, &NiagaraLightRendererProperties]}
  {emitterName=NE_WaveIdle, rendererClasses [&NiagaraMeshRendererProperties]}
  {emitterName=NE_DeployFlash, rendererClasses [&NiagaraSpriteRendererProperties]}
  {emitterName=NE_WaveDeploy, rendererClasses [&NiagaraMeshRendererProperties]}
  {emitterName=NE_HexGridDeploy, rendererClasses [&NiagaraMeshRendererProperties]}
  {emitterName=NE_HexSprites, rendererClasses [&NiagaraSpriteRendererProperties]}
 ]
 userVariables [
  User.bIsFrontEnd {type=NiagaraBool}
  User.bIsFrontEndPreview {type=NiagaraBool}
  User.ForwardDot {type=NiagaraFloat, defaultValue=0, description=""}
  User.RankedProgression {type=NiagaraInt32}
  User.RankFloat {type=NiagaraFloat, defaultValue=0, description=""}
  User.SmokeColor {type=LinearColor, description="", defaultValue {struct=LinearColor, value=<rgba 1 1 1 1>}}
 ]
}
